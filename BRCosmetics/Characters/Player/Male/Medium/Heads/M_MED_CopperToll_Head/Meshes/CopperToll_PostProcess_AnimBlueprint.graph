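# Post-process Anim Blueprint for the CopperToll player head that limits head rotation by partially countering it.
# Each update measures the head's delta from reference pose relative to neck_02, negates it and scales it per axis into a rotator.
# The scale partially counters a single axis and drops to zero during emotes.
# An additive bone-space Modify Bone applies the result to the head, with alpha enabled only when a pawn owner is valid.
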AnimBlueprint /BRCosmetics/Characters/Player/Male/Medium/Heads/M_MED_CopperToll_Head/Meshes/CopperToll_PostProcess_AnimBlueprint
target_skeleton=/Game/Characters/Player/Common/Fortnite_Base_Head/Fortnite_Base_Head_Export_Skeleton
graphs [
 AnimGraph {nodes [
   n1 = "Transform (Modify) Bone - Bone: head"(
    ComponentPose=Local To Component(LocalPose=Input Pose())
    Alpha=Get enableHeadRotLimitAlpha() -> (enableHeadRotLimitAlpha:double)
    Translation="0,0,0"
    Rotation=Get head rot() -> (head rot:Rotator)
    Scale="1,1,1"
   ) {
    rotationMode=BMM_Additive
    rotationSpace=BCS_BoneSpace
    alphaBoolBlend {blendOption=Linear}
    boneToModify {boneName=head}
    rotation=<rot 50 50 50>
   }
   Output Pose(Result=Component To Local(ComponentPose=@n1))
  ]}
 EventGraph {nodes [
   Event BlueprintUpdateAnimation()
   > n1 = Sequence!(execute) -> (then_0, then_1, then_2:Exec, then_3:Exec)
   > n2 = Sequence!(execute) -> (then_0, then_1, then_2, then_3, then_4:Exec, then_5:Exec, then_6:Exec)
   > n3 = Branch(Condition=Get bIsPlayingEmote())
   > n4 = Set headRotScale(headRotScale:Rotator="0, 0, 0") -> (Output_Get:Rotator)
   n5 = Set headRotScale(execute=@n3.else, headRotScale="0,0,0.6")
   n6 = BreakRotator(InRot=@n4)
   Set X percentage(execute=@n2.then_1, X percentage:float=@n6) -> (Output_Get:float)
   Set Y percentage(execute=@n2.then_2, Y percentage:float=@n6.Pitch) -> (Output_Get:float)
   Set Z percentage(execute=@n2.then_3, Z percentage:float=@n6.Yaw) -> (Output_Get:float)
   n7 = BreakRotator(InRot=@n5)
   Set X percentage(execute=@n2.then_4, X percentage=@n7)
   Set Y percentage(execute=@n2.then_5, Y percentage=@n7.Pitch)
   Set Z percentage(execute=@n2.then_6, Z percentage=@n7.Yaw)
   n8 = Is Valid(exec=@n1.then_2, InputObject=TryGetPawnOwner())
   Set enableHeadRotLimitAlpha(execute=@n8.Is Not Valid, enableHeadRotLimitAlpha:double=0) -> (Output_Get:double)
   Set enableHeadRotLimitAlpha(execute=@n8.Is Valid, enableHeadRotLimitAlpha=1)
   n9 = GetOwnerSkeletalMeshComponent()
   n10 = BreakRotator(InRot=BreakTransform(InTransform=GetDeltaTransformFromRefPose(self=@n9, BoneName=head, BaseName=neck_02)).Rotation)
   n11 = Get Y percentage() -> (Y percentage:float)
   n12 = Get X percentage() -> (X percentage:float)
   n13 = Get Z percentage() -> (Z percentage:float)
   > Set head rot(head rot:Rotator=MakeRotator(
     Roll=float * float!(A=@n10, B=-1, C:double=@n12) -> (ReturnValue)
     Pitch=float * float(A=@n10.Pitch, B=-1, C=@n11)
     Yaw=float * float(A=@n10.Yaw, B=-1, C=@n13)
    )) -> (Output_Get:Rotator)
   TryGetPawnOwner()
  ]}
]
variables [
 enableHeadRotLimitAlpha {type=number}
 head rot {type=Rotator}
 headRotScale {type=Rotator}
 X percentage {type=number}
 Y percentage {type=number}
 Z percentage {type=number}
]
